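; Creative move tool interaction behavior that deletes actors, with events for playing a delete animation and reacting to its completion.
; On the owning client, finished or directly destroyed actors join a deduplicated queue; the animation path also clears them from ReadyToBeDeleted.
; A single next-tick timer batches the queue into one ServerDestroyActors RPC on the PhoneToolActorTargetMode owner, then resets and signals the move tool that interaction finished.
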(event Custom|Destroy%20Actor
  (params (ActorToDestroy :type Actor%20Object%20Reference))
  ;[1 1 1] Destroy Actor With Authority
  (Actor|DestroyActor ActorToDestroy))

(event Custom|ServerPlayDeleteAnim
  ;[1 1 1] Play Delete Animation
  (CallFunction|PlayDeleteAnimation))

(event EventNotifyDestroyAnimationComplete
  (params (ForObject :type Object%20Reference))
  ;[1 1 1] Notifification event for the destroy animation complete
  (CallFunction|DestroyActoronAnimationFinished 0))

(fn BatchDestroyActorsOnServer
  (bind _asphone_tool_actor_target_mode (Utilities|Casting|CastToPhoneToolActorTargetMode (Components|GetOwner)))
  (Class|PhoneToolActorTargetMode|ServerDestroyActors _asphone_tool_actor_target_mode (Variables|Default|GetBatchActorsToDestroyOnServer))
  (Utilities|Array|Clear (Variables|Default|GetBatchActorsToDestroyOnServer))
  (Utilities|Time|ClearandInvalidateTimerbyHandle (Variables|Default|GetBatchDestroyOnServerTimerHandle))
  (Class|FortCreativeMoveTool|HandleInteractionFinished (CreativeInteraction|GetFortCreativeMoveTool)))

(fn DestroyActorOnAnimationFinished
  (params (ActorToDestroy :type Actor%20Object%20Reference))
  (if (CreativeMoveTool|IsRunningonOwningClient (CreativeInteraction|GetFortCreativeMoveTool))
    (CallFunction|QueueActorforDestroyonServer ActorToDestroy) ; This function can be called many times in a row,\nso queue up all the actors and send them all in a single server RPC
  )
  (Utilities|Array|RemoveItem (Variables|Default|GetReadyToBeDeleted) ActorToDestroy))

(fn DestroyAndDontRemove
  (params (ActorToDestroy :type Actor%20Object%20Reference))
  (bind _returnvalue (CreativeInteraction|GetFortCreativeMoveTool))
  (Utilities|IsValid _returnvalue
    (:Is%20Valid
      (if (CreativeMoveTool|IsRunningonOwningClient _returnvalue)
        (CallFunction|QueueActorforDestroyonServer ActorToDestroy) ; This function can be called many times in a row,\nso queue up all the actors and send them all in a single server RPC
      ))))

(fn QueueActorForDestroyOnServer
  (params (ActorToDestroy :type Actor%20Object%20Reference%20%28by%20ref%29))
  (Utilities|Array|AddUnique (Variables|Default|GetBatchActorsToDestroyOnServer) ActorToDestroy)
  ;[1 1 1] Only set the timer once, it will be reset when the event fires and the batched actors to destroy will be cleared. If there are more actors queued up after that, it will set a new timer.
  (if (not (Utilities|Time|IsValidTimerHandle (Variables|Default|GetBatchDestroyOnServerTimerHandle)))
    (bind _returnvalue (Utilities|Time|SetTimerforNextTickbyEvent (EventDispatchers|CreateEvent :!event BatchDestroyActorsOnServer)))
    (Variables|Default|SetBatchDestroyOnServerTimerHandle _returnvalue)))
